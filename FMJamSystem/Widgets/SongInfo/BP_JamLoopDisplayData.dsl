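; BP_JamLoopDisplayData is a display data object for one jam loop, used by the song info widgets of the FM Jam system.
; Its Init event packs title, artist, loop type, player net ID, loop instance ID and an auto-jammer profile picture flag into an SJamLoopDisplayData struct.
; It stores that struct in DisplayData and passes the result to CallOnDataChanged to notify listeners of the change.
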(event Custom|Init
  (params (Title :type Text) (Artist :type Text) (LoopType :type EFMJamLoopType%20Enum :default "Lead") (PlayerNetId :type Unique%20Net%20Id%20Structure) (LoopInstanceId :type Integer) (bUseAutoJammerProfilePic :type Boolean))
  (bind _output_get (Variables|Default|SetDisplayData (Utilities|Struct|MakeSJamLoopDisplayData Title Artist LoopType PlayerNetId LoopInstanceId bUseAutoJammerProfilePic)))
  (Default|CallOnDataChanged _output_get))
