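# Translucent UI material graph for an animated equalizer of SDF bars, with no connection to the material outputs shown.
# MF_UI_Pattern tiles the UVs into four columns, each drawn as a rounded box by MF_UI_Box.
# Column-quantized U coordinates pan across a blurred noise texture at a fixed V, giving each bar its own changing height.
# Parameters expose scroll speed, height amplitude and bar width.
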Material /FMJamSystem/Materials/M_SDF_EQ
graph {
 properties {blendMode=BLEND_Translucent, materialDomain=MD_UI, pixelDepthOffsetMode=PDOM_Legacy, usedShadingModels=Default Lit}
 nodes [
  Constant() {r=1}
  Constant() {r=100}
  n1 = Constant() {r=1}
  n2 = Constant() {r=4}
  n3 = ScalarParameter() {defaultValue=12, parameterName=Amplitude}
  n4 = Floor(None=@n2)
  n5 = Multiply(A=TextureCoordinate() {vTiling=0}, B=@n4)
  n6 = Panner(Coordinate=Divide(A=Floor(None=@n5), B=@n4), Speed=ScalarParameter() {defaultValue=0.7, parameterName=Speed}) {speedX=0.5}
  n7 = TextureSample(UVs=AppendVector(A=ComponentMask(None=@n6) {r=true}, B=Constant() {r=0.1}) {desc=Keep X Values}) {samplerType=SAMPLERTYPE_Grayscale, texture=/FortUILibrary/UIKit/Buttons/Textures/LowResBlurredNoise}
  MaterialFunctionCall(
   UVs=Subtract(
    A=MaterialFunctionCall!(UV, Tiling X=@n2, Tiling Y=@n1) {materialFunction=/FortUILibrary/MaterialFunctions/Basics/MF_UI_Pattern}.Scaled Pattern UVs
    B=Constant2Vector() {g=2}
   ) {constB=10}
   Size=AppendVector(A=ScalarParameter() {defaultValue=0.5, parameterName=Pipe Width}, B=LinearInterpolate(B=@n3, Alpha=@n7.RGB) {constB=4})
   CornerRoundness=Constant() {r=0.1}
   GlowMax=Constant() {r=0.01}
   KeepAspectRatio=StaticBool() {value=true}
  ) {materialFunction=&MF_UI_Box}
  TextureCoordinate()
 ]
}
