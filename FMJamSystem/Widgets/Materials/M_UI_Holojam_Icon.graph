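# Translucent UI material for the HoloJam icon in the FMJamSystem widgets.
# The icon texture is read as a distance field through MF_UI_AntiAliasing, and its Inner result is inverted.
# A lerp blends this with MF_UI_Circle's fill multiplied by dark indigo, using that tinted fill as both target and alpha.
# The raw circle fill also feeds a separate reroute.
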Material /FMJamSystem/Widgets/Materials/M_UI_Holojam_Icon
graph {
 properties {bAutomaticallySetUsageInEditor=false, blendMode=BLEND_Translucent, materialDomain=MD_UI, usedShadingModels=Default Lit}
 nodes [
  n1 = TextureSample() {samplerSource=SSM_Clamp_WorldGroupSettings, texture=/FMJamSystem/Widgets/Textures/T_UI_HoloJam_Icon}
  n2 = Constant() {r=0.5}
  n3 = Constant() {r=0.6}
  n4 = MaterialFunctionCall(Size=Constant() {r=0.92}) {materialFunction=&MF_UI_Circle}
  n5 = Multiply(A=Constant3Vector() {constant=<rgba 0.0243 0.0167 0.1146 1>}, B=@n4.Fill)
  LinearInterpolate(
   A=OneMinus(None=MaterialFunctionCall(DistanceField=@n1.RGB, Threshold=@n2, Smoothing=@n3) {materialFunction=&MF_UI_AntiAliasing}.Inner)
   B=@n5
   Alpha=@n5
  )
  Reroute(None=@n4.Fill)
 ]
}
