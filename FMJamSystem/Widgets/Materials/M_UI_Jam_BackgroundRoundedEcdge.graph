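# Translucent UI material for the Jam widgets that, going by its name, draws a rounded-edge background.
# A rounded-mask function with adjustable corner and pixel radius forms the base shape, from which a horizontally offset box fill tinted by PartColor is subtracted.
# A second, slightly shifted rounded mask with fixed corner radius is inverted as the lerp alpha, keeping that result only outside the shifted shape as an edge.
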Material /FMJamSystem/Widgets/Materials/M_UI_Jam_BackgroundRoundedEcdge
graph {
 properties {blendMode=BLEND_Translucent, materialDomain=MD_UI, pixelDepthOffsetMode=PDOM_Legacy, usedShadingModels=Default Lit}
 nodes [
  n1 = VectorParameter() {
   group=Shape
   parameterName=Corner Radius
   sortPriority=1
   channelNames {a=BottomRight, b=TopRight, g=BottomLeft, r=TopLeft}
   defaultValue=<rgba 12 12 12 12>
  }
  AppendVector(A=@n1, B=@n1)
  Constant4Vector()
  n2 = VectorParameter() {
   group=Shape
   parameterName=Corner Radius
   sortPriority=1
   channelNames {a=BottomRight, b=TopRight, g=BottomLeft, r=TopLeft}
   defaultValue=<rgba 12 12 12 12>
  }
  n3 = MaterialFunctionCall(
   Corner Radius=AppendVector(A=@n2, B=@n2)
   Radius in Pixels=ScalarParameter() {defaultValue=1, group=Shape, parameterName=Radius in Pixels, sliderMax=1, sortPriority=2}
  ) {materialFunction=&MF_UI_CoreUI_RoundedMask}
  n4 = TextureCoordinate()
  n5 = AppendVector(A=Constant() {r=0.01}, B=Constant())
  n6 = Multiply(
   A=VectorParameter() {parameterName=PartColor, defaultValue=<rgba 1 1 1 1>}.RGB
   B=MaterialFunctionCall(UVs=Add(A=@n4, B=@n5) {constB=-0.25}, Size=Constant2Vector() {g=1, r=1}) {materialFunction=&MF_UI_Box}.Fill
  )
  n7 = Constant() {r=0.006}
  n8 = Constant()
  n9 = MaterialFunctionCall(
   UV=Add(A=TextureCoordinate(), B=AppendVector(A=@n7, B=@n8)) {constB=-0.25}
   Corner Radius=Constant() {r=12}
   Radius in Pixels=ScalarParameter() {defaultValue=1, group=Shape, parameterName=Radius in Pixels, sliderMax=1, sortPriority=2}
  ) {materialFunction=&MF_UI_CoreUI_RoundedMask}
  LinearInterpolate(B=Subtract(A=Saturate(None=@n3.Shape Mask), B=Saturate(None=@n6)), Alpha=OneMinus(None=Saturate(None=@n9.Shape Mask)))
 ]
}
