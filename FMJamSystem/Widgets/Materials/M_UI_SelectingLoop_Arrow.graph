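# Translucent UI material that draws a tintable arrow, named for a selecting-loop widget in FMJamSystem.
# The arrow texture parameter is sampled through a Rotator fed a constant time, so it sits at a fixed rotation of roughly a quarter turn rather than animating.
# MF_Jam_AntiAliasing1 smooths its edges by threshold, and the result is multiplied separately by the Texture Color and Texture Opacity parameters.
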Material /FMJamSystem/Widgets/Materials/M_UI_SelectingLoop_Arrow
graph {
 properties {bAutomaticallySetUsageInEditor=false, blendMode=BLEND_Translucent, materialDomain=MD_UI, usedShadingModels=Default Lit}
 nodes [
  n1 = TextureCoordinate()
  n2 = Constant() {r=-6.26}
  n3 = MaterialFunctionCall!(
   Value=TextureSampleParameter2D(UVs=Rotator(Coordinate=@n1, Time=@n2)) {parameterName=Texture, texture=/FMJamSystem/Widgets/Textures/T_UI_Arrow}.RGB
   Threshold=Constant() {r=0.3}
   Smoothing=Constant() {r=2}
  ) {materialFunction=&/MF_Jam_AntiAliasing1}
  Multiply(A=@n3.EitherSide, B=ScalarParameter() {defaultValue=0.8, parameterName=Texture Opacity})
  Multiply(A=VectorParameter() {parameterName=Texture Color, defaultValue=<rgba 1 1 1 1>}.RGB, B=@n3.EitherSide)
 ]
}
