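# Background widget for the FM Jam loop display, layering a backplate image under a colored now-playing status bar.
# PreConstruct hides the bar and applies background opacity; Construct caches the backplate's dynamic material and tints the bar.
# SetBackgroundType recolors the material's underlayer for HoloJammer or default, while SetFocusVisual switches the backplate style and shows or hides the bar.
# SetNowPlayingColor tints the bar by loop type (Lead, Misc, Bass, Beat).
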WidgetBlueprint /FMJamSystem/Widgets/SongInfo/WBP_JamLoopDisplay_Background
graphs [
 EventGraph {nodes [
   n1 = Event PreConstruct()
   > Parent: PreConstruct(IsDesignTime=@n1.IsDesignTime)
   > SetVisibility(self=Get NowPlayingStatusBar() -> (NowPlayingStatusBar:Image Object Reference), InVisibility=Collapsed)
   > SetOpacity!(
    execute
    self:Image Object Reference=Get Background() -> (Background:Image Object Reference)
    InOpacity:float=Get BackgroundAlpha() -> (BackgroundAlpha:double)
   ) -> (then)
   Event Construct()
   > n2 = GetDynamicMaterial(self=Get Background())
   > Set MID_Background(MID_Background:Material Instance Dynamic Object Reference=@n2) -> (Output_Get:Material Instance Dynamic Object Reference)
   > SetColorAndOpacity(
    self:Image Object Reference=Get NowPlayingStatusBar()
    InColorAndOpacity:Linear Color Structure=Get NowPlaying_BG_Color() -> (NowPlaying_BG_Color:Linear Color Structure)
   )
  ]}
 SetBackgroundType {nodes [
   n1 = SetBackgroundType!() -> (then, IsHoloJammer:bool=~) {is=title K2Node_FunctionEntry}
   > n2 = Branch(Condition=@n1)
   > SetVectorParameterValue(
    self=Get MID_Background() -> (MID_Background:Material Instance Dynamic Object Reference)
    ParameterName=Underlayer_Color
    Value=Get HoloJammer_BG_Color() -> (HoloJammer_BG_Color:Linear Color Structure)
   )
   SetVectorParameterValue(
    execute=@n2.else
    self=Get MID_Background()
    ParameterName=Underlayer_Color
    Value=Get Default_BG_Color() -> (Default_BG_Color:Linear Color Structure)
   )
  ]}
 SetFocusVisual {nodes [
   n1 = SetFocusVisual!() -> (then, Focused:EJamMusicSlotFocusState Enum=NotFocused) {is=title K2Node_FunctionEntry}
   > n2 = Switch on EJamMusicSlotFocusState!(execute, Selection:EJamMusicSlotFocusState Enum=@n1) -> (NotFocused:Exec, FocusedAndEditable:Exec) {is=title K2Node_SwitchEnum}
   SetScalarParameterValue(execute=@n2.FocusedAndEditable, self=Get MID_Background(), ParameterName=Backplate Switch, Value=0)
   > SetVisibility(self=Get NowPlayingStatusBar(), InVisibility=SelfHitTestInvisible)
   SetScalarParameterValue(execute=@n2.NotFocused, self=Get MID_Background(), ParameterName=Backplate Switch, Value=3)
   > SetVisibility(self=Get NowPlayingStatusBar(), InVisibility=Collapsed)
  ]}
 SetNowPlayingColor {nodes [
   n1 = SetNowPlayingColor!() -> (then, NewParam:EFMJamLoopType Enum=Lead) {is=title K2Node_FunctionEntry}
   > n2 = Switch on EFMJamLoopType!(execute, Selection:EFMJamLoopType Enum=@n1) -> (Lead:Exec, Misc:Exec, Bass:Exec, Beat:Exec, Num:Exec) {is=title K2Node_SwitchEnum}
   > SetColorAndOpacity(self=Get NowPlayingStatusBar(), InColorAndOpacity=(R=1,G=0.7605,B=0.0356,A=1))
   SetColorAndOpacity(execute=@n2.Misc, self=Get NowPlayingStatusBar(), InColorAndOpacity=(R=0.9473,G=0.1248,B=0.4678,A=1))
   SetColorAndOpacity(execute=@n2.Bass, self=Get NowPlayingStatusBar(), InColorAndOpacity=(R=0.1878,G=1,B=0.1356,A=1))
   SetColorAndOpacity(execute=@n2.Num, self=Get NowPlayingStatusBar(), InColorAndOpacity=(R=1,G=1,B=1,A=0))
   SetColorAndOpacity(execute=@n2.Beat, self=Get NowPlayingStatusBar(), InColorAndOpacity=(R=0,G=0.552,B=0.8879,A=1))
  ]}
]
variables [
 BackgroundAlpha {type=number, default=1}
 Default_BG_Color {type=LinearColor, default=<rgba 0 0 0 0.8>}
 HoloJammer_BG_Color {type=LinearColor, default=<rgba 0 0.0523 0.151 0.6>}
 MID_Background {type=MaterialInstanceDynamic}
 NowPlaying_BG_Color {type=LinearColor, default=<rgba 0 0.0523 0.151 0.6>}
]
widgets {
 info {inheritedWidgetCount=0, parentClass=&UserWidget, rootWidgetClass=&GridPanel, widgetCount=3}
 widgets [
  GridPanel_22 {widgetClassPath=&GridPanel, widget=GridPanel_22, properties {columnFill [1], rowFill [1]}}
  Background {
   widgetClassPath=&Image
   bIsVariable=true
   parent=GridPanel_22
   widget=Background
   properties {
    brush {
     bIsDynamicallyLoaded=false
     drawAs=Image
     imageType=FullColor
     mirroring=NoMirror
     resourceName=None
     resourceObject=/SparksCore/Widgets/Materials/MI_UIKit_BackplateCombined
     tiling=NoTile
     imageSize=<32 32>
     margin=<ltrb 0 0 0 0>
     outlineSettings {
      bUseBrushTransparency=false
      roundingType=HalfHeightRadius
      width=0
      color {colorUseRule=UseColor_Specified, specifiedColor=<rgba 0 0 0 0>}
      cornerRadii=<0 0 0 0>
     }
     tintColor {colorUseRule=UseColor_Specified, specifiedColor=<rgba 1 1 1 1>}
     uVRegion {bIsValid=false, max=<0 0>, min=<0 0>}
    }
    colorAndOpacity=<rgba 1 1 1 0.8>
   }
   slot {class=GridSlot}
  }
  NowPlayingStatusBar {
   widgetClassPath=&Image
   bIsVariable=true
   parent=GridPanel_22
   widget=NowPlayingStatusBar
   properties {
    visibility=HitTestInvisible
    brush {
     bIsDynamicallyLoaded=false
     drawAs=Image
     imageType=FullColor
     mirroring=NoMirror
     resourceName=None
     resourceObject=/FMJamSystem/Widgets/Materials/M_UI_Jam_BackgroundRoundedEcdge
     tiling=NoTile
     imageSize=<32 32>
     margin=<ltrb 0 0 0 0>
     outlineSettings {
      bUseBrushTransparency=false
      roundingType=HalfHeightRadius
      width=0
      color {colorUseRule=UseColor_Specified, specifiedColor=<rgba 0 0 0 0>}
      cornerRadii=<0 0 0 0>
     }
     tintColor {colorUseRule=UseColor_Specified, specifiedColor=<rgba 1 1 1 1>}
     uVRegion {bIsValid=false, max=<0 0>, min=<0 0>}
    }
   }
   slot {class=GridSlot, properties {columnSpan=1, layer=1, rowSpan=1}}
  }
 ]
}
