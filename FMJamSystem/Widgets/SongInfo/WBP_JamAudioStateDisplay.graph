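# Jam audio state display widget for the Jam system, laid out as a fixed-width bar showing tempo and key/mode text.
# An overlay places BPM and key/mode text blocks side by side over a translucent Jam loop display background.
# UpdateTempo stores the passed BPM in Tempo and formats it into the BPM text.
# Key, Mode, Player Count and two text color variables are declared, but no graph here uses them.
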WidgetBlueprint /FMJamSystem/Widgets/SongInfo/WBP_JamAudioStateDisplay
graphs [
 EventGraph {links [], nodes []}
 UpdateTempo {nodes [
   n1 = UpdateTempo!() -> (then, BPM:int) {is=title K2Node_FunctionEntry}
   > n2 = Set Tempo(Tempo:int=@n1) -> (Output_Get:int=0)
   > "SetText (Text)"(
    self=Get BPMTextBlock() -> (BPMTextBlock:Text Block Object Reference)
    InText=Format Text!(Format="LOCTEXT(\"{bpm} BPM\")", bpm:int=@n2) -> (Result)
   )
  ]}
]
variables [
 Key {type=EMusicKey, default=Num}
 MaxPlayersTextColor {type=SlateColor, default {colorUseRule=UseColor_Specified, specifiedColor=<rgba 1 0.5104 0.5104 1>}}
 Mode {type=EMusicKeyMode, default=Num}
 Player Count {type=integer}
 StandardTextColor {type=SlateColor, default {colorUseRule=UseColor_Specified, specifiedColor=<rgba 1 1 1 1>}}
 Tempo {type=integer}
]
widgets {
 info {inheritedWidgetCount=0, parentClass=&UserWidget, rootWidgetClass=&Overlay, widgetCount=6}
 widgets [
  Overlay_1 {widgetClassPath=&Overlay, widget=Overlay_1}
  SizeBox_0 {
   widgetClassPath=&SizeBox
   parent=Overlay_1
   widget=SizeBox_0
   properties {bOverride_WidthOverride=true, widthOverride=425}
   slot {class=OverlaySlot, properties {horizontalAlignment=HAlign_Fill, verticalAlignment=VAlign_Fill}}
  }
  WBP_Background {
   widgetClassPath=&/WBP_JamLoopDisplay_Background
   bIsVariable=true
   parent=SizeBox_0
   widget=WBP_Background
   properties {backgroundAlpha=0.8, visibility=HitTestInvisible}
   slot {class=SizeBoxSlot}
  }
  HorizontalBox_24 {
   widgetClassPath=&HorizontalBox
   parent=Overlay_1
   widget=HorizontalBox_24
   slot {class=OverlaySlot, properties {horizontalAlignment=HAlign_Fill, verticalAlignment=VAlign_Fill, padding=<ltrb 18 10 18 10>}}
  }
  BPMTextBlock {
   widgetClassPath=&UEFN_TextBlock
   bIsVariable=true
   parent=HorizontalBox_24
   widget=BPMTextBlock
   properties {
    text="{bpm} BPM"
    font {
     bForceMonospaced=false
     bMaterialIsStencil=false
     fontMaterial=null
     fontObject=&HeadingNow
     letterSpacing=0
     monospacedWidth=1
     size=15.75
     skewAmount=0
     typefaceFontName=75 Medium
     outlineSettings {
      bApplyOutlineToDropShadows=false
      bMiteredCorners=false
      bSeparateFillAlpha=false
      outlineMaterial=null
      outlineSize=0
      outlineColor=<rgba 0 0 0 1>
     }
    }
   }
   slot {
    class=HorizontalBoxSlot
    properties {horizontalAlignment=HAlign_Left, verticalAlignment=VAlign_Center, padding=<ltrb 15 0 15 0>, size {sizeRule=Fill, value=1}}
   }
  }
  KeyModeTextBlock {
   widgetClassPath=&UEFN_TextBlock
   bIsVariable=true
   parent=HorizontalBox_24
   widget=KeyModeTextBlock
   properties {
    justification=Right
    text="{Key} {Mode}"
    font {
     bForceMonospaced=false
     bMaterialIsStencil=false
     fontMaterial=null
     fontObject=&HeadingNow
     letterSpacing=0
     monospacedWidth=1
     size=15.75
     skewAmount=0
     typefaceFontName=75 Medium
     outlineSettings {
      bApplyOutlineToDropShadows=false
      bMiteredCorners=false
      bSeparateFillAlpha=false
      outlineMaterial=null
      outlineSize=0
      outlineColor=<rgba 0 0 0 1>
     }
    }
   }
   slot {
    class=HorizontalBoxSlot
    properties {horizontalAlignment=HAlign_Right, verticalAlignment=VAlign_Center, padding=<ltrb 15 0 15 0>, size {sizeRule=Fill, value=1}}
   }
  }
 ]
}
