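# WBP_JamLoop_AlbumArt is a Jam Loop widget that shows a song's album art over a dark placeholder image.
# PreConstruct applies the BrushSize variable to both image brushes, so their size follows one setting.
# UpdateAlbumArtForSong hides the art for a None song name, leaving the placeholder visible.
# It then requests the art from the Sparks album art manager subsystem into a lazy image with texture releasing enabled.
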WidgetBlueprint /FMJamSystem/Widgets/WBP_JamLoop_AlbumArt
graphs [
 EventGraph {nodes [
   n1 = Event PreConstruct()
   > Parent: PreConstruct(IsDesignTime=@n1.IsDesignTime)
   n2 = Get BrushSize() -> (BrushSize:Vector 2D Structure)
   > Set members in Slate Brush!(
    execute:Exec
    StructRef:"Slate Brush Structure (by ref)"=Get Brush!(self:Image Object Reference=Get AlbumArt() -> (AlbumArt:Fort Lazy Image Object Reference)) -> (Brush:Slate Brush Structure)
    ImageSize:Vector 2D Structure=@n2
   ) -> (then:Exec, StructOut:"Slate Brush Structure (by ref)") {is=title K2Node_SetFieldsInStruct}
   > Set members in Slate Brush(StructRef=Get Brush(self=Get Placeholder() -> (Placeholder:Image Object Reference)), ImageSize=@n2)
   Event Construct()
   Event Tick()
  ]}
 UpdateAlbumArtForSong {nodes [
   n1 = UpdateAlbumArtForSong!() -> (then, SongShortName:Name) {is=title K2Node_FunctionEntry}
   > SetVisibility(
    self=Get AlbumArt()
    InVisibility="To Visibility (Boolean)"!(bIsVisible:bool="Equal (Name)"(A=@n1), TrueVisibility:ESlateVisibility Enum=Hidden, FalseVisibility:ESlateVisibility Enum=HitTestInvisible) -> (ReturnValue:ESlateVisibility Enum=Visible)
   )
   > RequestLazyAlbumArt!(
    execute
    AlbumArtManager:Sparks Album Art Manager Object Reference=Get SparksAlbumArtManager() -> (ReturnValue:Sparks Album Art Manager Object Reference) {is=title K2Node_GetSubsystem}
    SongShortName:"Name (by ref)"=Get SongShortName() -> (SongShortName:Name=None)
    LazyImage:Common Lazy Image Object Reference=Get AlbumArt()
   ) -> (then, ReturnValue:bool)
  ]}
]
variables [BrushSize {type=Vector2D, default=<48 48>}]
widgets {
 info {inheritedWidgetCount=0, parentClass=&UserWidget, rootWidgetClass=&Overlay, widgetCount=3}
 widgets [
  Overlay_0 {widgetClassPath=&Overlay, widget=Overlay_0}
  Placeholder {
   widgetClassPath=&Image
   bIsVariable=true
   parent=Overlay_0
   widget=Placeholder
   properties {
    brush {
     bIsDynamicallyLoaded=false
     drawAs=Image
     imageType=NoImage
     mirroring=NoMirror
     resourceName=None
     resourceObject=null
     tiling=NoTile
     imageSize=<48 48>
     margin=<ltrb 0 0 0 0>
     outlineSettings {
      bUseBrushTransparency=false
      roundingType=HalfHeightRadius
      width=0
      color {colorUseRule=UseColor_Specified, specifiedColor=<rgba 0 0 0 0>}
      cornerRadii=<0 0 0 0>
     }
     tintColor {colorUseRule=UseColor_Specified, specifiedColor=<rgba 1 1 1 1>}
     uVRegion {bIsValid=false, max=<0 0>, min=<0 0>}
    }
    colorAndOpacity=<rgba 0.06 0.06 0.06 1>
   }
   slot {class=OverlaySlot}
  }
  AlbumArt {
   widgetClassPath=&FortLazyImage
   bIsVariable=true
   parent=Overlay_0
   widget=AlbumArt
   properties {
    bEnableTextureReleasing=true
    brush {
     bIsDynamicallyLoaded=false
     drawAs=Image
     imageType=FullColor
     mirroring=NoMirror
     resourceName=None
     resourceObject=null
     tiling=NoTile
     imageSize=<48 48>
     margin=<ltrb 0 0 0 0>
     outlineSettings {
      bUseBrushTransparency=false
      roundingType=HalfHeightRadius
      width=0
      color {colorUseRule=UseColor_Specified, specifiedColor=<rgba 0 0 0 0>}
      cornerRadii=<0 0 0 0>
     }
     tintColor {colorUseRule=UseColor_Specified, specifiedColor=<rgba 1 1 1 1>}
     uVRegion {bIsValid=false, max=<0 0>, min=<0 0>}
    }
   }
   slot {class=OverlaySlot}
  }
 ]
}
