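# Translucent UI material for a rounded background, filed under the UIKit InputField system.
# It multiplies the Opacity output of the shared modular-button state lerp function (idle, hovered, focused, pressed, disabled) by a CoreUI rounded-rectangle shape mask.
# A Corner Radius vector parameter and a Radius in Pixels scalar, both in the Corners group, drive the mask's rounding.
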Material /FortUILibrary/Systems/InputField/Materials/M_UIKit_BackgroundRounded
graph {
 properties {
  bUsedWithStaticMesh=false
  blendMode=BLEND_Translucent
  materialDomain=MD_UI
  pixelDepthOffsetMode=PDOM_Legacy
  usedShadingModels=Default Lit
 }
 nodes [
  n1 = VectorParameter() {group=Corners, parameterName=Corner Radius, sortPriority=2, defaultValue=<rgba 0.25 0.25 0.25 0.25>}
  Multiply(
   A=MaterialFunctionCall!(Idle, Idle Opacity, Hovered, Hovered Opacity, Focused, Focused Opacity, Pressed, Pressed Opacity, Disabled, Disabled Opacity) {materialFunction=/FortUILibrary/Systems/ModularButton/MaterialFunctions/MF_UIKit_ButtonStatesBase_Lerp}.Opacity
   B=MaterialFunctionCall(
    Corner Radius=AppendVector(A=@n1.RGB, B=@n1.RGB)
    Radius in Pixels=ScalarParameter() {group=Corners, parameterName=Radius in Pixels, sliderMax=1, sortPriority=0}
   ) {materialFunction=&MF_UI_CoreUI_RoundedMask}.Shape Mask
  )
 ]
}
