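# Jam slot row widget showing a player's avatar, loop album art, song and artist, part icon and status overlay.
# Play state and user mode dim idle slots with placeholder lines and flag autojammers with a holojammer icon.
# Flagged, suppressed songs swap art and text for placeholder bars.
# Focus restyles text, background and selection arrow unless it is an on-card indicator; squad index tints the avatar's dynamic material.
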WidgetBlueprint /FMJamSystem/Widgets/SongInfo/WBP_JamLoopDisplay
graphs [
 DetermineIsHoloJammerByPlayStateAndUserMode {nodes [
   n1 = DetermineIsHoloJammerByPlayStateAndUserMode!() -> (then, PlayState:EJamMusicSlotVMPlayState Enum=SlotIdle, UserMode:EJamMusicSlotVMUserMode Enum=UserIsHuman) {is=title K2Node_FunctionEntry}
   > n2 = Switch on EJamMusicSlotVMPlayState!(execute, Selection:EJamMusicSlotVMPlayState Enum=@n1) -> (SlotIdle:Exec, SlotInUse:Exec) {is=title K2Node_SwitchEnum}
   > SetRenderOpacity(self=Get Overlay_Main() -> (Overlay_Main:Overlay Object Reference), InOpacity=0.6)
   > SetVisibility(
    self=[Get FakeLineTop() -> (FakeLineTop:Image Object Reference), Get FakeLineBottom() -> (FakeLineBottom:Image Object Reference)]
    InVisibility=HitTestInvisible
   )
   > SetVisibility(self=Get PartIconRight() -> (PartIconRight:BP Jam Part Icon Object Reference), InVisibility=Collapsed)
   > n3 = BP_ResetAvatarIcon(self:Sparks Social Avatar Icon Object Reference=Get PlayerAvatar() -> (PlayerAvatar:Sparks Social Avatar Icon Object Reference))
   SetVisibility(
    execute=[@n3.then, @n6.then]
    self=Get HolojammerIcon() -> (HolojammerIcon:Image Object Reference)
    InVisibility="To Visibility (Boolean)"!(
     bIsVisible:bool=Get HoloJammer() -> (HoloJammer:bool)
     TrueVisibility:ESlateVisibility Enum=SelfHitTestInvisible
     FalseVisibility:ESlateVisibility Enum=Hidden
    ) -> (ReturnValue:ESlateVisibility Enum=Visible)
   )
   > Return Node!(execute, IsHoloJammer:bool=Get HoloJammer())
   n4 = Switch on EJamMusicSlotVMUserMode!(execute=@n2.SlotInUse, Selection:EJamMusicSlotVMUserMode Enum=@n1.UserMode) -> (UserIsHuman:Exec, UserIsAutojammer:Exec) {is=title K2Node_SwitchEnum}
   n5 = Set HoloJammer(execute=@n4.UserIsAutojammer, HoloJammer:bool=true) -> (Output_Get:bool)
   SetRenderOpacity(execute=[@n5.then, @n4.UserIsHuman], self=Get Overlay_Main(), InOpacity=1)
   > SetVisibility(self=[Get FakeLineTop(), Get FakeLineBottom()], InVisibility=Collapsed)
   > n6 = SetVisibility(self=Get PartIconRight(), InVisibility=HitTestInvisible)
  ]}
 EventGraph {nodes [
   n1 = Process New Resolve Status!() -> (OutputDelegate, then, New Resolve State:EJamMusicSlotVMResolveState Enum=Unknown) {is=title K2Node_CustomEvent}
   > Clear and Invalidate Timer by Handle(Handle=Get Start Pending Appearance Handle() -> (Start Pending Appearance Handle:Timer Handle Structure))
   n2 = "Equal (Enum)"(A:EJamMusicSlotVMResolveState Enum=@n1.New Resolve State, B:EJamMusicSlotVMResolveState Enum=Flagged_Suppressed)
   > SetVisibility(
    self=[
     Get TextContent() -> (TextContent:Stack Box Object Reference)
     Get WBP_JamLoop_AlbumArt() -> (WBP_JamLoop_AlbumArt:WBP Jam Loop Album Art Object Reference)
    ]
    InVisibility=Select(Option 0=SelfHitTestInvisible, Option 1=Hidden, Index=@n2)
   )
   > SetVisibility(
    self=Get VerticalBox_FlaggedTextLines() -> (VerticalBox_FlaggedTextLines:Vertical Box Object Reference)
    InVisibility=Select(Option 0=Collapsed, Option 1=SelfHitTestInvisible, Index=@n2)
   )
   Start Pending Appearance() {is=title K2Node_CustomEvent}
   > StartPendingAnimatin(self:BP Jam Part Icon Object Reference=Get PartIconRight())
   n3 = Event PreConstruct()
   > Parent: PreConstruct(IsDesignTime=@n3.IsDesignTime)
   > SetVisibility(self=Get PlayerAvatar(), InVisibility=Get AvatarPortraitVisibility() -> (AvatarPortraitVisibility:ESlateVisibility Enum=Visible))
   > SetVisibility(
    self=Get WBP_Background() -> (WBP_Background:WBP Jam Loop Display Background Object Reference)
    InVisibility=Get Background_Visibility() -> (Background_Visibility:ESlateVisibility Enum=Visible)
   )
  ]}
 SetBackground {nodes [SetBackground() {is=title K2Node_FunctionEntry}]}
 SetFocusStatus {nodes [
   n1 = SetFocusStatus!() -> (then, Focused:EJamMusicSlotFocusState Enum=NotFocused) {is=title K2Node_FunctionEntry}
   > n2 = Branch(Condition=Get IsOnCardIndicator() -> (IsOnCardIndicator:bool))
   SetFocusVisual!(execute=@n2.else, self:WBP Jam Loop Display Background Object Reference=Get WBP_Background(), Focused:EJamMusicSlotFocusState Enum=@n1) -> (then)
   n3 = "Equal (Enum)"(A:EJamMusicSlotFocusState Enum=@n1, B:EJamMusicSlotFocusState Enum=NotFocused)
   > SetColorAndOpacity!(
    execute
    self=Get CommonTextBlock_SongName() -> (CommonTextBlock_SongName:Common Text Object Reference)
    InColorAndOpacity_SpecifiedColor:Linear Color Structure=Select(Option 0:Linear Color Structure=(R=0,G=0,B=0,A=1), Option 1:Linear Color Structure=(R=1,G=1,B=1,A=1), Index=@n3) -> (ReturnValue:Linear Color Structure)
    InColorAndOpacity_ColorUseRule:ESlateColorStylingMode Enum
   ) -> (then)
   > SetColorAndOpacity(
    self=Get CommonTextBlock_ArtistName() -> (CommonTextBlock_ArtistName:Common Text Object Reference)
    InColorAndOpacity_SpecifiedColor=Select(Option 0=(R=0,G=0,B=0,A=0.5), Option 1=(R=1,G=1,B=1,A=0.5), Index=@n3)
   )
   > SetVisibility(
    self=Get Arrow_SelectingLoop() -> (Arrow_SelectingLoop:Image Object Reference)
    InVisibility=Select(Option 0:ESlateVisibility Enum=HitTestInvisible, Option 1:ESlateVisibility Enum=Collapsed, Index=@n3) -> (ReturnValue:ESlateVisibility Enum)
   )
  ]}
 SetNowPlayingIconColor {nodes [
   n1 = SetNowPlayingIconColor!() -> (then, JamLoopType:EFMJamLoopType Enum=Lead) {is=title K2Node_FunctionEntry}
   > SetIconColorByJamLoopType!(execute, self:BP Jam Part Icon Object Reference=Get PartIconRight(), JamLoopType:EFMJamLoopType Enum=@n1) -> (then)
  ]}
 SetPlayerBGColorByNetId {nodes [
   n1 = SetPlayerBGColorByNetId!() -> (then, Player Net ID:Unique Net Id Structure) {is=title K2Node_FunctionEntry}
   > n2 = GetDynamicMaterial(self=Get PlayerAvatar())
   > SetVectorParameterValue(
    self=@n2
    ParameterName=BackgroundRGB
    Value=GetPinColorBySquadMemberIndex!(SquadIndex:int=GetSquadMemberIndexFromPlayerUniqueNetId!(PlayerNetId:"Unique Net Id Structure (by ref)"=@n1) -> (ReturnValue:int=0)) -> (ReturnValue:Linear Color Structure)
   )
  ]}
 UpdateAutojammerIconFromVM {nodes [UpdateAutojammerIconFromVM!() -> (then, The VM:Jam Music Slot VM Object Reference) {is=title K2Node_FunctionEntry}]}
]
variables [
 AvatarMaterialInstance {type=MaterialInstanceConstant, default=/Game/UI/SocialPanel/Art/Materials/UI_UserPortrait_Buffet}
 AvatarPortraitVisibility {type=ESlateVisibility, default=SelfHitTestInvisible}
 Background_Visibility {type=ESlateVisibility, default=HitTestInvisible}
 BackgroundColorAndOpacity {type=LinearColor, default=<rgba 0 0 0 0.8>}
 DarkTextColor {type=SlateColor, default {colorUseRule=UseColor_Specified, specifiedColor=<rgba 0 0 0 1>}}
 Default_BG_Color {type=LinearColor, default=<rgba 0 0 0 0.8>}
 HoloJammer_BG_Color {type=LinearColor, default=<rgba 0 0 0 0.8>}
 HoloJammer_BGColor {type=LinearColor, default=<rgba 0 0 0 0>}
 IsOnCardIndicator {type=boolean}
 LightTextColor {type=SlateColor, default {colorUseRule=UseColor_Specified, specifiedColor=<rgba 1 1 1 1>}}
 MID_Background {type=MaterialInstanceDynamic}
 Start Pending Appearance Handle {type=TimerHandle}
]
widgets {
 info {inheritedWidgetCount=0, parentClass=&UserWidget, rootWidgetClass=&SizeBox, widgetCount=30}
 widgets [
  SizeBox_0 {widgetClassPath=&SizeBox, widget=SizeBox_0, properties {bOverride_WidthOverride=true, widthOverride=371}}
  Overlay_Main {widgetClassPath=&Overlay, bIsVariable=true, parent=SizeBox_0, widget=Overlay_Main, slot {class=SizeBoxSlot}}
  Overlay_0 {
   widgetClassPath=&Overlay
   parent=Overlay_Main
   widget=Overlay_0
   slot {class=OverlaySlot, properties {horizontalAlignment=HAlign_Fill, verticalAlignment=VAlign_Fill}}
  }
  WBP_Background {
   widgetClassPath=&/WBP_JamLoopDisplay_Background
   bIsVariable=true
   parent=Overlay_0
   widget=WBP_Background
   properties {backgroundAlpha=0.8, renderOpacity=0.8, visibility=HitTestInvisible}
   slot {class=OverlaySlot, properties {horizontalAlignment=HAlign_Fill, verticalAlignment=VAlign_Fill}}
  }
  TextBorder {
   widgetClassPath=&Overlay
   parent=Overlay_Main
   widget=TextBorder
   slot {class=OverlaySlot, properties {horizontalAlignment=HAlign_Fill, verticalAlignment=VAlign_Fill}}
  }
  PlayerDisplay {
   widgetClassPath=&Overlay
   parent=TextBorder
   widget=PlayerDisplay
   slot {class=OverlaySlot, properties {horizontalAlignment=HAlign_Fill, verticalAlignment=VAlign_Fill}}
  }
  PlayerDisplayBox {
   widgetClassPath=&HorizontalBox
   parent=PlayerDisplay
   widget=PlayerDisplayBox
   slot {class=OverlaySlot, properties {horizontalAlignment=HAlign_Fill, verticalAlignment=VAlign_Fill, padding=<ltrb 0 4 0 4>}}
  }
  SizeBox_54 {
   widgetClassPath=&SizeBox
   parent=PlayerDisplayBox
   widget=SizeBox_54
   properties {bOverride_MinDesiredHeight=true, bOverride_MinDesiredWidth=true, minDesiredHeight=52, minDesiredWidth=52}
   slot {class=HorizontalBoxSlot, properties {padding=<ltrb 8 0 0 0>}}
  }
  GridPanel_0 {widgetClassPath=&GridPanel, parent=SizeBox_54, widget=GridPanel_0, slot {class=SizeBoxSlot}}
  PlayerAvatar {
   widgetClassPath=/Script/SparksCoreUI.SparksSocialAvatarIcon
   bIsVariable=true
   parent=GridPanel_0
   widget=PlayerAvatar
   properties {
    avatarMaterial=&UI_UserPortrait_Buffet
    brush {
     bIsDynamicallyLoaded=false
     drawAs=Image
     imageType=FullColor
     mirroring=NoMirror
     resourceName=None
     resourceObject=&UI_UserPortrait_Buffet
     tiling=NoTile
     imageSize=<52 52>
     margin=<ltrb 0 0 0 0>
     outlineSettings {
      bUseBrushTransparency=false
      roundingType=HalfHeightRadius
      width=0
      color {colorUseRule=UseColor_Specified, specifiedColor=<rgba 0 0 0 0>}
      cornerRadii=<0 0 0 0>
     }
     tintColor {colorUseRule=UseColor_Specified, specifiedColor=<rgba 1 1 1 1>}
     uVRegion {bIsValid=false, max=<0 0>, min=<0 0>}
    }
   }
   slot {class=GridSlot}
  }
  HolojammerIcon {
   widgetClassPath=&Image
   bIsVariable=true
   parent=GridPanel_0
   widget=HolojammerIcon
   properties {
    visibility=SelfHitTestInvisible
    brush {
     bIsDynamicallyLoaded=false
     drawAs=Image
     imageType=FullColor
     mirroring=NoMirror
     resourceName=None
     resourceObject=/FMJamSystem/Widgets/Materials/M_UI_Holojam_Icon
     tiling=NoTile
     imageSize=<20 20>
     margin=<ltrb 0 0 0 0>
     outlineSettings {
      bUseBrushTransparency=false
      roundingType=HalfHeightRadius
      width=0
      color {colorUseRule=UseColor_Specified, specifiedColor=<rgba 0 0 0 0>}
      cornerRadii=<0 0 0 0>
     }
     tintColor {colorUseRule=UseColor_Specified, specifiedColor=<rgba 1 1 1 1>}
     uVRegion {bIsValid=false, max=<0 0>, min=<0 0>}
    }
    renderTransform {angle=0, scale=<1 1>, shear=<0 0>, translation=<18 16>}
   }
   slot {class=GridSlot, properties {horizontalAlignment=HAlign_Center, layer=1, verticalAlignment=VAlign_Center, nudge=<-4 -4>}}
  }
  Overlay_1 {
   widgetClassPath=&Overlay
   parent=PlayerDisplayBox
   widget=Overlay_1
   slot {class=HorizontalBoxSlot, properties {verticalAlignment=VAlign_Center, padding=<ltrb 12 0 0 0>}}
  }
  SizeBox_198 {
   widgetClassPath=&SizeBox
   parent=Overlay_1
   widget=SizeBox_198
   properties {bOverride_MinDesiredHeight=true, bOverride_MinDesiredWidth=true, minDesiredHeight=48, minDesiredWidth=48}
   slot {class=OverlaySlot, properties {horizontalAlignment=HAlign_Fill, verticalAlignment=VAlign_Fill}}
  }
  WBP_JamLoop_AlbumArt {
   widgetClassPath=/FMJamSystem/Widgets/WBP_JamLoop_AlbumArt
   bIsVariable=true
   parent=Overlay_1
   widget=WBP_JamLoop_AlbumArt
   properties {brushSize=<42 42>}
   slot {class=OverlaySlot, properties {horizontalAlignment=HAlign_Center, verticalAlignment=VAlign_Center}}
  }
  VerticalBox_FakeTextLines {
   widgetClassPath=&VerticalBox
   parent=PlayerDisplayBox
   widget=VerticalBox_FakeTextLines
   slot {class=HorizontalBoxSlot, properties {verticalAlignment=VAlign_Center, padding=<ltrb 8 0 0 0>}}
  }
  FakeLineTop {
   widgetClassPath=&Image
   bIsVariable=true
   parent=VerticalBox_FakeTextLines
   widget=FakeLineTop
   properties {
    brush {
     bIsDynamicallyLoaded=false
     drawAs=Image
     imageType=NoImage
     mirroring=NoMirror
     resourceName=None
     resourceObject=null
     tiling=NoTile
     imageSize=<180 8>
     margin=<ltrb 0 0 0 0>
     outlineSettings {
      bUseBrushTransparency=false
      roundingType=HalfHeightRadius
      width=0
      color {colorUseRule=UseColor_Specified, specifiedColor=<rgba 0 0 0 0>}
      cornerRadii=<0 0 0 0>
     }
     tintColor {colorUseRule=UseColor_Specified, specifiedColor=<rgba 1 1 1 1>}
     uVRegion {bIsValid=false, max=<0 0>, min=<0 0>}
    }
    colorAndOpacity=<rgba 1 1 1 0.2>
   }
   slot {class=VerticalBoxSlot, properties {horizontalAlignment=HAlign_Left, padding=<ltrb 0 0 0 8>}}
  }
  FakeLineBottom {
   widgetClassPath=&Image
   bIsVariable=true
   parent=VerticalBox_FakeTextLines
   widget=FakeLineBottom
   properties {
    brush {
     bIsDynamicallyLoaded=false
     drawAs=Image
     imageType=NoImage
     mirroring=NoMirror
     resourceName=None
     resourceObject=null
     tiling=NoTile
     imageSize=<120 8>
     margin=<ltrb 0 0 0 0>
     outlineSettings {
      bUseBrushTransparency=false
      roundingType=HalfHeightRadius
      width=0
      color {colorUseRule=UseColor_Specified, specifiedColor=<rgba 0 0 0 0>}
      cornerRadii=<0 0 0 0>
     }
     tintColor {colorUseRule=UseColor_Specified, specifiedColor=<rgba 1 1 1 1>}
     uVRegion {bIsValid=false, max=<0 0>, min=<0 0>}
    }
    colorAndOpacity=<rgba 1 1 1 0.2>
   }
   slot {class=VerticalBoxSlot, properties {horizontalAlignment=HAlign_Left}}
  }
  VerticalBox_FlaggedTextLines {
   widgetClassPath=&VerticalBox
   bIsVariable=true
   parent=PlayerDisplayBox
   widget=VerticalBox_FlaggedTextLines
   properties {visibility=Collapsed}
   slot {class=HorizontalBoxSlot, properties {verticalAlignment=VAlign_Center, padding=<ltrb 8 0 0 0>}}
  }
  FlaggedLineTop {
   widgetClassPath=&Image
   bIsVariable=true
   parent=VerticalBox_FlaggedTextLines
   widget=FlaggedLineTop
   properties {
    brush {
     bIsDynamicallyLoaded=false
     drawAs=Image
     imageType=NoImage
     mirroring=NoMirror
     resourceName=None
     resourceObject=null
     tiling=NoTile
     imageSize=<180 8>
     margin=<ltrb 0 0 0 0>
     outlineSettings {
      bUseBrushTransparency=false
      roundingType=HalfHeightRadius
      width=0
      color {colorUseRule=UseColor_Specified, specifiedColor=<rgba 0 0 0 0>}
      cornerRadii=<0 0 0 0>
     }
     tintColor {colorUseRule=UseColor_Specified, specifiedColor=<rgba 1 1 1 1>}
     uVRegion {bIsValid=false, max=<0 0>, min=<0 0>}
    }
    colorAndOpacity=<rgba 1 1 1 0.2>
   }
   slot {class=VerticalBoxSlot, properties {horizontalAlignment=HAlign_Left, padding=<ltrb 0 0 0 8>}}
  }
  FlaggedLineBottom {
   widgetClassPath=&Image
   bIsVariable=true
   parent=VerticalBox_FlaggedTextLines
   widget=FlaggedLineBottom
   properties {
    brush {
     bIsDynamicallyLoaded=false
     drawAs=Image
     imageType=NoImage
     mirroring=NoMirror
     resourceName=None
     resourceObject=null
     tiling=NoTile
     imageSize=<120 8>
     margin=<ltrb 0 0 0 0>
     outlineSettings {
      bUseBrushTransparency=false
      roundingType=HalfHeightRadius
      width=0
      color {colorUseRule=UseColor_Specified, specifiedColor=<rgba 0 0 0 0>}
      cornerRadii=<0 0 0 0>
     }
     tintColor {colorUseRule=UseColor_Specified, specifiedColor=<rgba 1 1 1 1>}
     uVRegion {bIsValid=false, max=<0 0>, min=<0 0>}
    }
    colorAndOpacity=<rgba 1 1 1 0.2>
   }
   slot {class=VerticalBoxSlot, properties {horizontalAlignment=HAlign_Left}}
  }
  TextContent {
   widgetClassPath=&StackBox
   bIsVariable=true
   parent=PlayerDisplayBox
   widget=TextContent
   properties {orientation=Orient_Vertical, visibility=Hidden}
   slot {class=HorizontalBoxSlot, properties {horizontalAlignment=HAlign_Left, verticalAlignment=VAlign_Center, size {sizeRule=Fill, value=1}}}
  }
  CommonTextBlock_SongName {
   widgetClassPath=&CommonTextBlock
   bIsVariable=true
   parent=TextContent
   widget=CommonTextBlock_SongName
   properties {
    bOverride_Cursor=true
    clipping=ClipToBounds
    justification=Center
    scrollStyle=&TextScrollStyle-Base_C
    text="{Song Title}"
    visibility=Visible
    font {
     bForceMonospaced=false
     bMaterialIsStencil=false
     fontMaterial=null
     fontObject=&HeadingNow
     letterSpacing=0
     monospacedWidth=1
     size=13.5
     skewAmount=0
     typefaceFontName=75 Medium
     outlineSettings {
      bApplyOutlineToDropShadows=false
      bMiteredCorners=false
      bSeparateFillAlpha=false
      outlineMaterial=null
      outlineSize=0
      outlineColor=<rgba 0 0 0 1>
     }
    }
    shadowOffset=<0 0>
   }
   slot {class=StackBoxSlot, properties {horizontalAlignment=HAlign_Left, size {sizeRule=Fill, value=1}}}
  }
  CommonTextBlock_ArtistName {
   widgetClassPath=&CommonTextBlock
   bIsVariable=true
   parent=TextContent
   widget=CommonTextBlock_ArtistName
   properties {
    bOverride_Cursor=true
    clipping=ClipToBounds
    justification=Center
    scrollStyle=&TextScrollStyle-Base_C
    text="{Artist Name}"
    colorAndOpacity {colorUseRule=UseColor_Specified, specifiedColor=<rgba 1 1 1 0.5>}
    font {
     bForceMonospaced=false
     bMaterialIsStencil=false
     fontMaterial=null
     fontObject=&HeadingNow
     letterSpacing=0
     monospacedWidth=1
     size=10.5
     skewAmount=0
     typefaceFontName=74 Regular
     outlineSettings {
      bApplyOutlineToDropShadows=false
      bMiteredCorners=false
      bSeparateFillAlpha=false
      outlineMaterial=null
      outlineSize=0
      outlineColor=<rgba 0 0 0 1>
     }
    }
    shadowOffset=<0 0>
   }
   slot {class=StackBoxSlot, properties {horizontalAlignment=HAlign_Left}}
  }
  Icon {
   widgetClassPath=&Overlay
   bIsVariable=true
   parent=PlayerDisplayBox
   widget=Icon
   slot {class=HorizontalBoxSlot, properties {horizontalAlignment=HAlign_Right, verticalAlignment=VAlign_Center, padding=<ltrb 12 0 12 0>}}
  }
  Current_Track_BG {
   widgetClassPath=&Image
   bIsVariable=true
   parent=Icon
   widget=Current_Track_BG
   properties {
    brush {
     bIsDynamicallyLoaded=false
     drawAs=Image
     imageType=FullColor
     mirroring=NoMirror
     resourceName=None
     resourceObject=/FortUILibrary/Systems/InputField/MaterialInstances/MI_UIKit_BackgroundRounded
     tiling=NoTile
     imageSize=<32 32>
     margin=<ltrb 0 0 0 0>
     outlineSettings {
      bUseBrushTransparency=false
      roundingType=HalfHeightRadius
      width=0
      color {colorUseRule=UseColor_Specified, specifiedColor=<rgba 0 0 0 0>}
      cornerRadii=<0 0 0 0>
     }
     tintColor {colorUseRule=UseColor_Specified, specifiedColor=<rgba 1 1 1 1>}
     uVRegion {bIsValid=false, max=<0 0>, min=<0 0>}
    }
    colorAndOpacity=<rgba 0 0 0 0.6>
   }
   slot {class=OverlaySlot, properties {horizontalAlignment=HAlign_Fill, verticalAlignment=VAlign_Fill}}
  }
  PartIconRight {
   widgetClassPath=/FMJamSystem/Widgets/PartSwitcher/BP_JamPartIcon
   bIsVariable=true
   parent=Icon
   widget=PartIconRight
   properties {
    gradient=/FMJamSystem/Materials/MI_UI_Icon_Instrument_Type_SDF_LessLines
    foregroundColor {colorUseRule=UseColor_Specified, specifiedColor=<rgba 1 0 1 1>}
    imageSize=<36 36>
   }
   slot {class=OverlaySlot}
  }
  SlotStateLayer {
   widgetClassPath=&Overlay
   parent=Overlay_Main
   widget=SlotStateLayer
   slot {class=OverlaySlot, properties {horizontalAlignment=HAlign_Fill, verticalAlignment=VAlign_Fill}}
  }
  StatusOverlaidFillBar {
   widgetClassPath=/Script/UMG.ProgressBar
   bIsVariable=true
   parent=SlotStateLayer
   widget=StatusOverlaidFillBar
   properties {
    barFillType=FillFromCenter
    fillColorAndOpacity=<rgba 0.4156 0 0.0613 0>
    widgetStyle {
     enableFillAnimation=false
     backgroundImage {
      bIsDynamicallyLoaded=false
      drawAs=NoDrawType
      imageType=FullColor
      mirroring=NoMirror
      resourceName=../../../Engine/Content/Slate/Common/ProgressBar_Background.png
      resourceObject=null
      tiling=NoTile
      imageSize=<0 0>
      margin=<ltrb 0.4167 0.4167 0.4167 0.4167>
      outlineSettings {
       bUseBrushTransparency=false
       roundingType=HalfHeightRadius
       width=0
       color {colorUseRule=UseColor_Specified, specifiedColor=<rgba 0 0 0 0>}
       cornerRadii=<0 0 0 0>
      }
      tintColor {colorUseRule=UseColor_Specified, specifiedColor=<rgba 1 1 1 1>}
      uVRegion {bIsValid=false, max=<0 0>, min=<0 0>}
     }
     fillImage {
      bIsDynamicallyLoaded=false
      drawAs=Image
      imageType=FullColor
      mirroring=NoMirror
      resourceName=None
      resourceObject=/FortUILibrary/Systems/InputField/Materials/M_UIKit_BackgroundRounded
      tiling=NoTile
      imageSize=<0 0>
      margin=<ltrb 0.4167 0.4167 0.4167 0.4167>
      outlineSettings {
       bUseBrushTransparency=false
       roundingType=HalfHeightRadius
       width=0
       color {colorUseRule=UseColor_Specified, specifiedColor=<rgba 0 0 0 0>}
       cornerRadii=<0 0 0 0>
      }
      tintColor {colorUseRule=UseColor_Specified, specifiedColor=<rgba 1 1 1 1>}
      uVRegion {bIsValid=false, max=<0 0>, min=<0 0>}
     }
     marqueeImage {
      bIsDynamicallyLoaded=false
      drawAs=Image
      imageType=FullColor
      mirroring=NoMirror
      resourceName=None
      resourceObject=null
      tiling=Horizontal
      imageSize=<20 12>
      margin=<ltrb 0 0 0 0>
      outlineSettings {
       bUseBrushTransparency=false
       roundingType=HalfHeightRadius
       width=0
       color {colorUseRule=UseColor_Specified, specifiedColor=<rgba 0 0 0 0>}
       cornerRadii=<0 0 0 0>
      }
      tintColor {colorUseRule=UseColor_Specified, specifiedColor=<rgba 1 1 1 1>}
      uVRegion {bIsValid=false, max=<0 0>, min=<0 0>}
     }
    }
   }
   slot {class=OverlaySlot, properties {horizontalAlignment=HAlign_Fill, verticalAlignment=VAlign_Fill}}
  }
  StatusOverlayMessage {
   widgetClassPath=&CommonTextBlock
   bIsVariable=true
   parent=SlotStateLayer
   widget=StatusOverlayMessage
   properties {
    bAutoCollapseWithEmptyText=true
    colorAndOpacity {colorUseRule=UseColor_Specified, specifiedColor=<rgba 1 1 1 -1>}
    font {
     bForceMonospaced=false
     bMaterialIsStencil=false
     fontMaterial=null
     fontObject=&HeadingNow
     letterSpacing=0
     monospacedWidth=1
     size=15
     skewAmount=0.064
     typefaceFontName=75 Medium Italic
     outlineSettings {
      bApplyOutlineToDropShadows=false
      bMiteredCorners=false
      bSeparateFillAlpha=false
      outlineMaterial=null
      outlineSize=0
      outlineColor=<rgba 0 0 0 1>
     }
    }
    shadowOffset=<0 -0.4779>
   }
   slot {class=OverlaySlot, properties {horizontalAlignment=HAlign_Right, verticalAlignment=VAlign_Bottom, padding=<ltrb 0 0 15 0>}}
  }
  Arrow_SelectingLoop {
   widgetClassPath=&Image
   bIsVariable=true
   parent=Overlay_Main
   widget=Arrow_SelectingLoop
   properties {
    visibility=HitTestInvisible
    brush {
     bIsDynamicallyLoaded=false
     drawAs=Image
     imageType=FullColor
     mirroring=NoMirror
     resourceName=None
     resourceObject=/FMJamSystem/Widgets/Materials/M_UI_SelectingLoop_Arrow
     tiling=NoTile
     imageSize=<32 32>
     margin=<ltrb 0 0 0 0>
     outlineSettings {
      bUseBrushTransparency=false
      roundingType=HalfHeightRadius
      width=0
      color {colorUseRule=UseColor_Specified, specifiedColor=<rgba 0 0 0 0>}
      cornerRadii=<0 0 0 0>
     }
     tintColor {colorUseRule=UseColor_Specified, specifiedColor=<rgba 1 1 1 1>}
     uVRegion {bIsValid=false, max=<0 0>, min=<0 0>}
    }
   }
   slot {class=OverlaySlot, properties {horizontalAlignment=HAlign_Right, verticalAlignment=VAlign_Center, padding=<ltrb 0 0 -35 0>}}
  }
 ]
}
