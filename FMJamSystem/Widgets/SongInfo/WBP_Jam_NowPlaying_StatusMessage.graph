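# Status message banner widget for Jam's Now Playing UI, showing either a loop-limit warning or the Jam's privacy level.
# An overlay stacks a red rounded background behind a placeholder icon and a pink-tinted, auto-wrapping text block.
# Two custom events only set that text: ShowLoopLimitMessage uses a fixed variable, while ShowJamPrivacyMessage formats a base phrase with the passed-in privacy text.
# The graph contains no show, hide or restyle logic.
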WidgetBlueprint /FMJamSystem/Widgets/SongInfo/WBP_Jam_NowPlaying_StatusMessage
graphs [EventGraph {nodes [
   ShowLoopLimitMessage() {is=title K2Node_CustomEvent}
   > "SetText (Text)"(self=Get MessageTextBlock() -> (MessageTextBlock:Text Block Object Reference), InText=Get LoopLimitText() -> (LoopLimitText:Text))
   n1 = ShowJamPrivacyMessage!() -> (OutputDelegate, then, PrivacyLevel:Text) {is=title K2Node_CustomEvent}
   > "SetText (Text)"(
    self=Get MessageTextBlock()
    InText=Format Text!(
     Format="LOCTEXT(\"{BaseMessage} {Privacy}\")"
     BaseMessage:Text=Get PrivacyLevelText() -> (PrivacyLevelText:Text)
     Privacy:Text=@n1.PrivacyLevel
    ) -> (Result)
   )
  ]}]
variables [
 LoopLimitText {type=string, default=Loop limit reached for this Jam}
 PrivacyLevelText {type=string, default=This Jam is}
]
widgets {
 info {inheritedWidgetCount=0, parentClass=&UserWidget, rootWidgetClass=&Overlay, widgetCount=5}
 widgets [
  Overlay_1 {widgetClassPath=&Overlay, widget=Overlay_1}
  Background {
   widgetClassPath=&Image
   bIsVariable=true
   parent=Overlay_1
   widget=Background
   properties {
    visibility=SelfHitTestInvisible
    brush {
     bIsDynamicallyLoaded=false
     drawAs=Image
     imageType=NoImage
     mirroring=NoMirror
     resourceName=None
     resourceObject=/FortUILibrary/Systems/InputField/MaterialInstances/MI_UIKit_BackgroundRounded
     tiling=NoTile
     imageSize=<450 48>
     margin=<ltrb 0 0 0 0>
     outlineSettings {
      bUseBrushTransparency=false
      roundingType=HalfHeightRadius
      width=0
      color {colorUseRule=UseColor_Specified, specifiedColor=<rgba 0 0 0 0>}
      cornerRadii=<0 0 0 0>
     }
     tintColor {colorUseRule=UseColor_Specified, specifiedColor=<rgba 1 1 1 1>}
     uVRegion {bIsValid=false, max=<0 0>, min=<0 0>}
    }
    colorAndOpacity=<rgba 1 0 0 1>
   }
   slot {class=OverlaySlot, properties {horizontalAlignment=HAlign_Fill, verticalAlignment=VAlign_Fill}}
  }
  HorizontalBox_24 {
   widgetClassPath=&HorizontalBox
   parent=Overlay_1
   widget=HorizontalBox_24
   slot {class=OverlaySlot, properties {horizontalAlignment=HAlign_Center, verticalAlignment=VAlign_Center, padding=<ltrb 5 0 5 0>}}
  }
  Iconplaceholder {
   widgetClassPath=&Image
   bIsVariable=true
   parent=HorizontalBox_24
   widget=Iconplaceholder
   properties {brush {
     bIsDynamicallyLoaded=false
     drawAs=Image
     imageType=NoImage
     mirroring=NoMirror
     resourceName=None
     resourceObject=/Game/UI/Foundation/Textures/Buttons/T-SlotIcon-NotRecyclable-64
     tiling=NoTile
     imageSize=<32 32>
     margin=<ltrb 0 0 0 0>
     outlineSettings {
      bUseBrushTransparency=false
      roundingType=HalfHeightRadius
      width=0
      color {colorUseRule=UseColor_Specified, specifiedColor=<rgba 0 0 0 0>}
      cornerRadii=<0 0 0 0>
     }
     tintColor {colorUseRule=UseColor_Specified, specifiedColor=<rgba 1 0.5089 0.5089 1>}
     uVRegion {bIsValid=false, max=<0 0>, min=<0 0>}
    }}
   slot {class=HorizontalBoxSlot, properties {horizontalAlignment=HAlign_Right, verticalAlignment=VAlign_Center}}
  }
  MessageTextBlock {
   widgetClassPath=&UEFN_TextBlock
   bIsVariable=true
   parent=HorizontalBox_24
   widget=MessageTextBlock
   properties {
    autoWrapText=true
    text=Loop limit reached for this Jam
    colorAndOpacity {colorUseRule=UseColor_Specified, specifiedColor=<rgba 1 0.5104 0.5104 1>}
    font {
     bForceMonospaced=false
     bMaterialIsStencil=false
     fontMaterial=null
     fontObject=&HeadingNow
     letterSpacing=0
     monospacedWidth=1
     size=13.5
     skewAmount=0
     typefaceFontName=74 Regular
     outlineSettings {
      bApplyOutlineToDropShadows=false
      bMiteredCorners=false
      bSeparateFillAlpha=false
      outlineMaterial=null
      outlineSize=0
      outlineColor=<rgba 0 0 0 1>
     }
    }
   }
   slot {class=HorizontalBoxSlot, properties {verticalAlignment=VAlign_Center, padding=<ltrb 3 4 4 4>, size {sizeRule=Fill, value=1}}}
  }
 ]
}
