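# Personal jam loop "Now Playing" widget for the FM Jam System, showing a song title and instrument part icon.
# An EQ material image, labeled title and BP_JamPartIcon share a row over a collapsed highlight background.
# SetValues writes Title and passes LoopType to the icon, leaving its Artist input unused.
# HighlightFadeOut fades the background's alpha, but nothing here plays it or uses the color and part-list variables.
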WidgetBlueprint /FMJamSystem/Widgets/SongInfo/WBP_PersonalJamLoopDisplay
animations [HighlightFadeOut {
  binding_count=1
  length_seconds=1
  object=&self:HighlightFadeOut
  sequence {bindings [HighlightBackground {
     id=#1
     possessed_class=Image
     tracks [{
       type=MovieSceneColorTrack
       display_name=Color and Opacity
       sections [{
         type=MovieSceneColorSection
         end_frame=20
         start_frame=0
         channels [
          Color.R {type=MovieSceneScriptingFloatChannel, const [1,1]}
          Color.G {type=MovieSceneScriptingFloatChannel, const [1,1]}
          Color.B {type=MovieSceneScriptingFloatChannel, const [1,1]}
          Color.A {type=MovieSceneScriptingFloatChannel, keys [[0,1],[20,0]]}
         ]
        }]
      }]
    }]}
 }]
graphs [
 EventGraph {links [], nodes []}
 SetValues {nodes [
   n1 = SetValues!() -> (then, Title:Text, Artist:Text, LoopType:EFMJamLoopType Enum=Lead) {is=title K2Node_FunctionEntry}
   > "SetText (Text)"(self=Get SongTitleTextBlock() -> (SongTitleTextBlock:Text Block Object Reference), InText=@n1)
   > Set Active Icon by LoopType!(
    execute
    self:BP Jam Part Icon Object Reference=Get PartIconRight() -> (PartIconRight:BP Jam Part Icon Object Reference)
    LoopType:EFMJamLoopType Enum=@n1.LoopType
   ) -> (then) {is=title K2Node_CallFunction}
  ]}
]
variables [
 DarkTextColor {type=SlateColor, default {colorUseRule=UseColor_Specified, specifiedColor=<rgba 0 0 0 1>}}
 LightTextColor {type=SlateColor, default {colorUseRule=UseColor_Specified, specifiedColor=<rgba 1 1 1 1>}}
 PartList {type="array<string>", default [Bass,Drum,Lead,Vox]}
 Tint Color {
  type="array<SlateColor>"
  default [
   {colorUseRule=UseColor_Specified, specifiedColor=<rgba 0.4706 0.7333 0.4275 1>}
   {colorUseRule=UseColor_Specified, specifiedColor=<rgba 0.2824 0.6392 0.898 1>}
   {colorUseRule=UseColor_Specified, specifiedColor=<rgba 0.8275 0.3686 0.5608 1>}
   {colorUseRule=UseColor_Specified, specifiedColor=<rgba 0.8275 0.7647 0.1882 1>}
  ]
 }
]
widgets {
 info {inheritedWidgetCount=0, parentClass=&UserWidget, rootWidgetClass=&CommonBorder, widgetCount=13}
 widgets [
  PersonalJamBorder {
   widgetClassPath=&CommonBorder
   bIsVariable=true
   widget=PersonalJamBorder
   properties {
    background {
     bIsDynamicallyLoaded=false
     drawAs=NoDrawType
     imageType=NoImage
     mirroring=NoMirror
     resourceName=None
     resourceObject=null
     tiling=NoTile
     imageSize=<32 32>
     margin=<ltrb 0 0 0 0>
     outlineSettings {
      bUseBrushTransparency=false
      roundingType=HalfHeightRadius
      width=0
      color {colorUseRule=UseColor_Specified, specifiedColor=<rgba 0 0 0 0>}
      cornerRadii=<0 0 0 0>
     }
     tintColor {colorUseRule=UseColor_Specified, specifiedColor=<rgba 1 1 1 0.8>}
     uVRegion {bIsValid=false, max=<0 0>, min=<0 0>}
    }
    padding=<ltrb 0 0 0 0>
   }
  }
  Overlay_1 {
   widgetClassPath=&Overlay
   parent=PersonalJamBorder
   widget=Overlay_1
   slot {class=BorderSlot, properties {padding=<ltrb 0 0 0 0>}}
  }
  HighlightBackground {
   widgetClassPath=&Image
   bIsVariable=true
   parent=Overlay_1
   widget=HighlightBackground
   properties {
    visibility=Collapsed
    brush {
     bIsDynamicallyLoaded=false
     drawAs=Image
     imageType=NoImage
     mirroring=NoMirror
     resourceName=None
     resourceObject=/FortUILibrary/Systems/InputField/Materials/M_UIKit_BackgroundRounded
     tiling=NoTile
     imageSize=<450 85>
     margin=<ltrb 0 0 0 0>
     outlineSettings {
      bUseBrushTransparency=false
      roundingType=HalfHeightRadius
      width=0
      color {colorUseRule=UseColor_Specified, specifiedColor=<rgba 0 0 0 0>}
      cornerRadii=<0 0 0 0>
     }
     tintColor {colorUseRule=UseColor_Specified, specifiedColor=<rgba 1 1 1 1>}
     uVRegion {bIsValid=false, max=<0 0>, min=<0 0>}
    }
    colorAndOpacity=<rgba 0 0 0 0.6>
   }
   slot {class=OverlaySlot, properties {horizontalAlignment=HAlign_Fill, verticalAlignment=VAlign_Fill}}
  }
  HorizontalBox_175 {
   widgetClassPath=&HorizontalBox
   parent=Overlay_1
   widget=HorizontalBox_175
   slot {class=OverlaySlot, properties {horizontalAlignment=HAlign_Fill, verticalAlignment=VAlign_Center, padding=<ltrb 10 10 10 10>}}
  }
  Overlay_0 {
   widgetClassPath=&Overlay
   parent=HorizontalBox_175
   widget=Overlay_0
   slot {class=HorizontalBoxSlot, properties {horizontalAlignment=HAlign_Left, verticalAlignment=VAlign_Center, padding=<ltrb 18 0 0 0>}}
  }
  EQ_Bar_Animated {
   widgetClassPath=&HorizontalBox
   parent=Overlay_0
   widget=EQ_Bar_Animated
   slot {class=OverlaySlot, properties {horizontalAlignment=HAlign_Center, verticalAlignment=VAlign_Bottom}}
  }
  Image_48 {
   widgetClassPath=&Image
   bIsVariable=true
   parent=EQ_Bar_Animated
   widget=Image_48
   properties {brush {
     bIsDynamicallyLoaded=false
     drawAs=Image
     imageType=NoImage
     mirroring=NoMirror
     resourceName=None
     resourceObject=/FMJamSystem/Materials/M_SDF_EQ
     tiling=NoTile
     imageSize=<32 32>
     margin=<ltrb 0 0 0 0>
     outlineSettings {
      bUseBrushTransparency=false
      roundingType=HalfHeightRadius
      width=0
      color {colorUseRule=UseColor_Specified, specifiedColor=<rgba 0 0 0 0>}
      cornerRadii=<0 0 0 0>
     }
     tintColor {colorUseRule=UseColor_Specified, specifiedColor=<rgba 1 1 1 1>}
     uVRegion {bIsValid=false, max=<0 0>, min=<0 0>}
    }}
   slot {class=HorizontalBoxSlot}
  }
  VerticalBox_51 {
   widgetClassPath=&VerticalBox
   parent=HorizontalBox_175
   widget=VerticalBox_51
   slot {class=HorizontalBoxSlot, properties {padding=<ltrb 12 0 0 0>, size {sizeRule=Fill, value=1}}}
  }
  UEFN_TextBlock_C_155 {
   widgetClassPath=&UEFN_TextBlock
   parent=VerticalBox_51
   widget=UEFN_TextBlock_C_155
   properties {
    text=Now Playing
    visibility=SelfHitTestInvisible
    colorAndOpacity {colorUseRule=UseColor_Specified, specifiedColor=<rgba 1 1 1 0.5>}
    font {
     bForceMonospaced=false
     bMaterialIsStencil=false
     fontMaterial=null
     fontObject=&HeadingNow
     letterSpacing=0
     monospacedWidth=1
     size=15
     skewAmount=0
     typefaceFontName=74 Regular
     outlineSettings {
      bApplyOutlineToDropShadows=false
      bMiteredCorners=false
      bSeparateFillAlpha=false
      outlineMaterial=null
      outlineSize=0
      outlineColor=<rgba 0 0 0 1>
     }
    }
   }
   slot {class=VerticalBoxSlot}
  }
  Spacer_1 {widgetClassPath=&Spacer, parent=VerticalBox_51, widget=Spacer_1, properties {size=<0 5>}, slot {class=VerticalBoxSlot}}
  SongTitleTextBlock {
   widgetClassPath=&UEFN_TextBlock
   bIsVariable=true
   parent=VerticalBox_51
   widget=SongTitleTextBlock
   properties {
    text="{SongTitle}"
    font {
     bForceMonospaced=false
     bMaterialIsStencil=false
     fontMaterial=null
     fontObject=&HeadingNow
     letterSpacing=0
     monospacedWidth=1
     size=15
     skewAmount=0
     typefaceFontName=75 Medium
     outlineSettings {
      bApplyOutlineToDropShadows=false
      bMiteredCorners=false
      bSeparateFillAlpha=false
      outlineMaterial=null
      outlineSize=0
      outlineColor=<rgba 0 0 0 1>
     }
    }
   }
   slot {class=VerticalBoxSlot, properties {horizontalAlignment=HAlign_Left, size {sizeRule=Fill, value=1}}}
  }
  Spacer {
   widgetClassPath=&Spacer
   parent=HorizontalBox_175
   widget=Spacer
   properties {size=<40 40>}
   slot {class=HorizontalBoxSlot, properties {horizontalAlignment=HAlign_Right}}
  }
  PartIconRight {
   widgetClassPath=/FMJamSystem/Widgets/PartSwitcher/BP_JamPartIcon
   bIsVariable=true
   parent=HorizontalBox_175
   widget=PartIconRight
   properties {
    foregroundColor {colorUseRule=UseColor_Specified, specifiedColor=<rgba 1 0 1 1>}
    renderTransform {angle=0, scale=<1.75 1.75>, shear=<0 0>, translation=<-18 0>}
   }
   slot {class=HorizontalBoxSlot, properties {horizontalAlignment=HAlign_Right, verticalAlignment=VAlign_Center}}
  }
 ]
}
